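# Sound cue for the XP level-up UI sound in gameplay, playing a single wave player with Fort_XP_LevelUp_01.
# It runs in the critical in-game UI sound class at top priority with boosted volume and a very large max distance.
# Instance limiting references the shared LevelUp_Priority_Conc concurrency set.
# A pre-effect send to the AB_Sidechain_FortMusic_Vol audio bus provides a sidechain signal for music volume.
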SoundCue /Game/Sounds/Fort_UI_Sounds/Gameplay/XP/Fort_XP_LevelUp_Cue
root=WavePlayer_2
nodes [{id=WavePlayer_2, properties {soundWaveAssetPtr=&/Fort_XP_LevelUp_01}}]
properties {
 maxDistance=2097152
 priority=100
 soundClassObject=&Fort_UI_InGame_Critical
 volumeMultiplier=2
 concurrencyOverrides {maxCount=1, resolutionRule=StopOldest, volumeScale=0.75, volumeScaleAttackTime=0.2}
 concurrencySet [&/LevelUp_Priority_Conc]
 preEffectBusSends [{audioBus=&AB_Sidechain_FortMusic_Vol, soundSourceBus=null}]
}
