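# Sound cue for the XP medal UI sound in gameplay, playing a single medal wave.
# One wave player feeds the output directly, with no randomization or modulation.
# It uses the critical in-game UI sound class at slightly reduced volume.
# A concurrency override caps how many copies play at once, and a very large max distance is set.
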SoundCue /Game/Sounds/Fort_UI_Sounds/Gameplay/XP/Fort_XP_Medal_Cue
root=WavePlayer_0
nodes [{id=WavePlayer_0, properties {soundWaveAssetPtr=&/S11XP_Medal_01}}]
properties {
 bOverrideConcurrency=true
 maxDistance=2097152
 soundClassObject=&Fort_UI_InGame_Critical
 volumeMultiplier=0.78
 concurrencyOverrides {maxCount=2}
}
