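# Sound cue for the XP popup in gameplay UI, routed to the in-game critical UI sound class.
# A fixed short delay feeds a random node that picks one of three equally weighted popup wave variations.
# The cue raises volume above default and overrides concurrency to allow at most two simultaneous instances.
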SoundCue /Game/Sounds/Fort_UI_Sounds/Gameplay/XP/Fort_XP_Popup_Cue
root=Delay_0
nodes [
 {id=Delay_0, children [Random_1], properties {delayMax=0.2, delayMin=0.2}}
 {id=Random_1, children [WavePlayer_3,WavePlayer_4,WavePlayer_5], properties {weights [1,1,1]}}
 {id=WavePlayer_3, properties {soundWaveAssetPtr=&/S11XP_Popup_01}}
 {id=WavePlayer_4, properties {soundWaveAssetPtr=&/S11XP_Popup_02}}
 {id=WavePlayer_5, properties {soundWaveAssetPtr=&/S11XP_Popup_03}}
]
properties {
 bOverrideConcurrency=true
 maxDistance=2097152
 soundClassObject=&Fort_UI_InGame_Critical
 volumeMultiplier=1.5
 concurrencyOverrides {maxCount=2}
}
