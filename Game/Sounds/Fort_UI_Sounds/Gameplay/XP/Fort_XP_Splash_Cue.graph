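# Sound cue for the XP splash UI sound played during gameplay, assigned to the critical in-game UI sound class.
# A Random node with equal weights picks one of three alternate splash wave players on each play.
# It references a dedicated XP splash concurrency group and also carries override settings for a single voice that stops the oldest.
# The max distance is very large and the volume is slightly boosted.
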SoundCue /Game/Sounds/Fort_UI_Sounds/Gameplay/XP/Fort_XP_Splash_Cue
root=Random_1
nodes [
 {id=Random_1, children [WavePlayer_0,WavePlayer_4,WavePlayer_5], properties {weights [1,1,1]}}
 {id=WavePlayer_0, properties {soundWaveAssetPtr=&/S11XP_Splash_Alt_01}}
 {id=WavePlayer_4, properties {soundWaveAssetPtr=&/S11XP_Splash_Alt_02}}
 {id=WavePlayer_5, properties {soundWaveAssetPtr=&/S11XP_Splash_Alt_03}}
]
properties {
 maxDistance=2097152
 soundClassObject=&Fort_UI_InGame_Critical
 volumeMultiplier=1.062
 concurrencyOverrides {maxCount=1, resolutionRule=StopOldest}
 concurrencySet [&/Fort_XP_Splash_Concurrency_Group]
}
