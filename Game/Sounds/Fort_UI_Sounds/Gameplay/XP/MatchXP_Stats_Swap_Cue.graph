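# UI sound cue for the Match XP stats swap, played on the Fort_UI sound class.
# It reuses the player survey pop-up wave, raised in pitch by the cue multiplier and slightly randomized by a Modulator.
# A fixed short Delay at the root holds back playback.
# A concurrency override limits it to two simultaneous instances, stopping the oldest.
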SoundCue /Game/Sounds/Fort_UI_Sounds/Gameplay/XP/MatchXP_Stats_Swap_Cue
root=Delay_0
nodes [
 {id=Delay_0, children [Modulator_0], properties {delayMax=0.1, delayMin=0.1}}
 {id=Modulator_0, children [WavePlayer_1], properties {pitchMax=1.01, pitchMin=0.98}}
 {id=WavePlayer_1, properties {soundWaveAssetPtr=/Game/Sounds/Fort_UI_Sounds/FrontEnd/Fort_UI_PlayerSurvey_PopUp_01}}
]
properties {
 bOverrideConcurrency=true
 maxDistance=2097152
 pitchMultiplier=1.5
 soundClassObject=&Fort_UI
 volumeMultiplier=0.8
 concurrencyOverrides {maxCount=2, resolutionRule=StopOldest}
}
