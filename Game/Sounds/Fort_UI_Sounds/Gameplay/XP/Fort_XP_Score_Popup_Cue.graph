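# Sound cue for the Fortnite UI XP score popup, played as gameplay UI feedback.
# A random node picks one of three equally weighted popup wave variations.
# It plays in the Fort_UI sound class at reduced volume, with overridden concurrency that caps simultaneous instances and stops the oldest first.
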SoundCue /Game/Sounds/Fort_UI_Sounds/Gameplay/XP/Fort_XP_Score_Popup_Cue
root=Random_0
nodes [
 {id=Random_0, children [WavePlayer_1,WavePlayer_2,WavePlayer_3], properties {weights [1,1,1]}}
 {id=WavePlayer_1, properties {soundWaveAssetPtr=&/Fort_XP_Score_Popup_01}}
 {id=WavePlayer_2, properties {soundWaveAssetPtr=&/Fort_XP_Score_Popup_02}}
 {id=WavePlayer_3, properties {soundWaveAssetPtr=&/Fort_XP_Score_Popup_03}}
]
properties {
 bOverrideConcurrency=true
 maxDistance=2097152
 soundClassObject=&Fort_UI
 volumeMultiplier=0.4
 concurrencyOverrides {maxCount=2, resolutionRule=StopOldest}
}
